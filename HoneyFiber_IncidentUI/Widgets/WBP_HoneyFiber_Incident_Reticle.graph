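# Fortnite weapon reticle extension widget for HoneyFiber Incident that plays an equip animation when constructed.
# An Overlay stacks two images using the same reticle material instance.
# Anim_Equip drives material rotation and size plus render opacity, so the main reticle twists, grows and fades in while the second image grows larger, flashes and fades out.
# Anim_Activated has no bindings and is never played by the graph.
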WidgetBlueprint /HoneyFiber_IncidentUI/Widgets/WBP_HoneyFiber_Incident_Reticle
animations [
 Anim_Equip {
  binding_count=2
  length_seconds=0.45
  object=&self:Anim_Equip
  sequence {bindings [
    Image_Reticle {
     id=#1
     possessed_class=Image
     tracks [
      {
       type=MovieSceneWidgetMaterialTrack
       sections [{
         type=MovieSceneParameterSection
         end_frame=9
         start_frame=0
         channels [
          Switch {type=MovieSceneScriptingFloatChannel, const [1,1]}
          ShakeAmmount {type=MovieSceneScriptingFloatChannel, const [0,1]}
          ReticleRotation {type=MovieSceneScriptingFloatChannel, keys [[0,-0.12],[9,0]]}
          ReticleSize {type=MovieSceneScriptingFloatChannel, keys [[0,0.5],[9,1]]}
         ]
        }]
      }
      {
       type=MovieSceneFloatTrack
       display_name=Render Opacity
       sections [{type=MovieSceneFloatSection, end_frame=9, start_frame=4, channels [None {type=MovieSceneScriptingFloatChannel, keys [[4,0],[9,1]]}]}]
      }
     ]
    }
    Reticle_Animation {
     id=#2
     possessed_class=Image
     tracks [
      {
       type=MovieSceneWidgetMaterialTrack
       sections [{
         type=MovieSceneParameterSection
         end_frame=9
         start_frame=0
         channels [
          Switch {type=MovieSceneScriptingFloatChannel, const [0,1]}
          ShakeAmmount {type=MovieSceneScriptingFloatChannel, const [0,1]}
          Progressss {type=MovieSceneScriptingFloatChannel, const [1,1]}
          ReticleRotation {type=MovieSceneScriptingFloatChannel, keys [[0,-0.12],[8.9997,0]]}
          ReticleSize {type=MovieSceneScriptingFloatChannel, keys [[0,0.5],[8.9997,1.2]]}
         ]
        }]
      }
      {
       type=MovieSceneFloatTrack
       display_name=Render Opacity
       sections [{
         type=MovieSceneFloatSection
         end_frame=9
         start_frame=0
         channels [None {type=MovieSceneScriptingFloatChannel, keys [[0,0],[6,1],[8.9997,0]]}]
        }]
      }
     ]
    }
   ]}
 }
 Anim_Activated {
  binding_count=0
  length_seconds=5
  object=&self:Anim_Activated
  sequence {bindings []}
 }
]
graphs [EventGraph {nodes [
   Event Construct()
   > PlayAnimation(InAnimation=Get Anim_Equip() -> (Anim_Equip:Widget Animation Object Reference))
  ]}]
variables []
widgets {
 info {
  inheritedWidgetCount=1
  parentClass=/Script/FortniteUI.FortWeaponReticleExtensionWidgetBase
  rootWidgetClass=&Overlay
  widgetCount=4
 }
 widgets [
  Overlay_0 {widgetClassPath=&Overlay, widget=Overlay_0}
  Image_Reticle {
   widgetClassPath=&Image
   bIsVariable=true
   parent=Overlay_0
   widget=Image_Reticle
   properties {brush {
     bIsDynamicallyLoaded=false
     drawAs=Image
     imageType=FullColor
     mirroring=NoMirror
     resourceName=None
     resourceObject=/HoneyFiber_IncidentUI/Art/Materials/MI_HoneyFiber_Incident_Reticle
     tiling=NoTile
     imageSize=<80 80>
     margin=<ltrb 0 0 0 0>
     outlineSettings {
      bUseBrushTransparency=false
      roundingType=HalfHeightRadius
      width=0
      color {colorUseRule=UseColor_Specified, specifiedColor=<rgba 0 0 0 0>}
      cornerRadii=<0 0 0 0>
     }
     tintColor {colorUseRule=UseColor_Specified, specifiedColor=<rgba 1 1 1 1>}
     uVRegion {bIsValid=false, max=<0 0>, min=<0 0>}
    }}
   slot {class=OverlaySlot}
  }
  Reticle_Animation {
   widgetClassPath=&Image
   bIsVariable=true
   parent=Overlay_0
   widget=Reticle_Animation
   properties {brush {
     bIsDynamicallyLoaded=false
     drawAs=Image
     imageType=FullColor
     mirroring=NoMirror
     resourceName=None
     resourceObject=&MI_HoneyFiber_Incident_Reticle
     tiling=NoTile
     imageSize=<80 80>
     margin=<ltrb 0 0 0 0>
     outlineSettings {
      bUseBrushTransparency=false
      roundingType=HalfHeightRadius
      width=0
      color {colorUseRule=UseColor_Specified, specifiedColor=<rgba 0 0 0 0>}
      cornerRadii=<0 0 0 0>
     }
     tintColor {colorUseRule=UseColor_Specified, specifiedColor=<rgba 1 1 1 1>}
     uVRegion {bIsValid=false, max=<0 0>, min=<0 0>}
    }}
   slot {class=OverlaySlot}
  }
  TutorialHighlight {widgetClassPath=&TutorialHighlight}
 ]
}
